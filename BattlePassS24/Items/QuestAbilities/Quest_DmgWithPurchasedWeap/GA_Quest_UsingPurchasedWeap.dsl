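; Passive Battle Pass quest ability that credits damage dealt to tagged opponents with weapons bought through NPC services or vending machines.
; A one-second timer snapshots the inventory, pausing when a weapon-tagged NPC service event fires.
; New items are found by diffing against that snapshot, immediately for valid NPC services, otherwise by polling for delayed delivery.
; Damage events whose current weapon matches a recorded definition send the magnitude as a quest stat.
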(event Ability|EventFailedToActivatePassiveAbility
  (params (ActorInfo :type Gameplay%20Ability%20Actor%20Info%20Structure))
  (|CallParentFunction ActorInfo))

(event Custom|Inventory%20Query%20Timer
  ;[1 1 1] Checking the inventory so that we can have a snapshot of it before the NPC services are used. Upgrading weapons doesn't behave the same as purchasing so this is needed.
  (CallFunction|MakePlayerInventoryItemArray))

(event Ability|EventActivateAbility
  (|CallParentFunction)
  (bind (_asynctaskproxy _payload) (Ability|Tasks|WaitGameplayEvent "(TagName=\"Event.NPCService.Used\")" 0 false false)
    (:then
      ;[1 0.4249 0.395] For Future use: This is what you would change if you wanted to do X with a purchased weapon
      (bind (_asynctaskproxy_2 _payload_2) (Ability|Tasks|WaitGameplayEvent "(TagName=\"Event.Damage.Dealt\")" 0 false false)
        (:then
          ;[1 1 1] Checking the inventory so that we can have a snapshot of it before the NPC services are used. Upgrading weapons doesn't behave the same as purchasing so this is needed.
          (Utilities|FlowControl|DoOnce false
            (:Completed
              (bind _returnvalue (Utilities|Time|SetTimerbyEvent (delegate Custom|Inventory%20Query%20Timer) 1.0 true))
              (Variables|Default|SetTimer_InventoryQueryHandle _returnvalue))))
        (:EventReceived
          ;;[1 1 1] Make sure we're hitting an opponent
          (bind (_payload_eventtag _payload_instigator _payload_target _payload_optionalobject _payload_optionalobject2 _payload_contexthandle _payload_instigatortags _payload_targettags _payload_eventmagnitude) (Utilities|Struct|BreakGameplayEventData _payload_2))
          (bind _asfort_player_pawn (Utilities|Casting|CastToFortPlayerPawn _payload_target))
          (if (GameplayTags|HasAnyMatchingGameplayTags _asfort_player_pawn (Variables|Default|GetRequired%20Opponent%20Tags))
            ;;[1 1 1] Check to see if the current weapon being used is one of the ones we've paid for.
            (if (Utilities|Array|ContainsItem (Variables|Default|GetPaidForWeapons) (Item|GetItemDefinition (FortWeapon|GetInventoryItem (Weapon|GetCurrentWeapon (Variables|GAVars|Data|GetOwningPlayerPawn)))))
              (Quest|SendComplexCustomStatEvent (Variables|GAVars|Data|GetOwningPlayerQuestManager) 0 "(GameplayTags=((TagName=\"Athena.Quests.DmgWithPurchasedWeap\")))" "(GameplayTags=)" (Math|Float|Truncate _payload_eventmagnitude)))))))
    (:EventReceived
      ;[1 1 1] If the thing we bought was a weapon, remember what was in our inventory then start a timer to check for new items to be added. Doing this because there's a long delay between purchases and when the thing actually gets into your inventory.
      (bind (_payload_eventtag_2 _payload_instigator_2 _payload_target_2 _payload_optionalobject_2 _payload_optionalobject2_2 _payload_contexthandle_2 _payload_instigatortags_2) (Utilities|Struct|BreakGameplayEventData _payload))
      (if (GameplayTags|HasAnyTags _payload_instigatortags_2 (Variables|Default|GetT_RequiredWeaponTags))
        (Utilities|Time|PauseTimerbyHandle (Variables|Default|GetTimer_InventoryQueryHandle))
        (if (GameplayTags|HasAnyTags _payload_instigatortags_2 (Variables|Default|GetValid%20NPCServices) true)
          ;[1 1 1] Check if the thing we bought is already in the inventory previously.
          (bind _items (InventoryManagement|GetItemsfromInventoryOwner (Fortnite|GetFortPlayerControllerfromActor (Variables|GAVars|Data|GetOwningPlayerPawn))))
          (if (Utilities|Array|IsNotEmpty _items)
            (for _array_element _items
              (if (not (Utilities|Array|ContainsItem (Variables|Default|GetInventory%20Items) _array_element))
                ;[1 1 1] Add the new thing to the Paid for Weapons array, remember this new inventory, and stop the timer that runs the check.
                (Utilities|Array|AddUnique (Variables|Default|GetPaidForWeapons) (Item|GetItemDefinition _array_element))
                (CallFunction|MakePlayerInventoryItemArray)
                (Utilities|Time|ClearandInvalidateTimerbyHandle (Variables|Default|GetTimer_VendingMachineHandle))
                (Utilities|Time|UnpauseTimerbyHandle (Variables|Default|GetTimer_InventoryQueryHandle)))))
          (else
            (CallFunction|MakePlayerInventoryItemArray)
            (bind _returnvalue_2 (Utilities|Time|SetTimerbyEvent (delegate Custom|Remember%20Purchased%20Weapons%20-%20Vending%20or%20NPC) 1.0 true))
            (Variables|Default|SetTimer_VendingMachineHandle _returnvalue_2)))))))

(event Custom|Remember%20Purchased%20Weapons%20-%20Vending%20or%20NPC
  ;[1 1 1] Check if the thing we bought is already in the inventory previously.
  (bind _items (InventoryManagement|GetItemsfromInventoryOwner (Fortnite|GetFortPlayerControllerfromActor (Variables|GAVars|Data|GetOwningPlayerPawn))))
  (if (Utilities|Array|IsNotEmpty _items)
    (for _array_element _items
      (if (not (Utilities|Array|ContainsItem (Variables|Default|GetInventory%20Items) _array_element))
        ;[1 1 1] Add the new thing to the Paid for Weapons array, remember this new inventory, and stop the timer that runs the check.
        (Utilities|Array|AddUnique (Variables|Default|GetPaidForWeapons) (Item|GetItemDefinition _array_element))
        (CallFunction|MakePlayerInventoryItemArray)
        (Utilities|Time|ClearandInvalidateTimerbyHandle (Variables|Default|GetTimer_VendingMachineHandle))
        (Utilities|Time|UnpauseTimerbyHandle (Variables|Default|GetTimer_InventoryQueryHandle))))))

(event Custom|Make%20Player%20Inventory%20item%20Array
  ;[1 1 1] Remember what's in my inventory right now
  (bind _items (InventoryManagement|GetItemsfromInventoryOwner (Fortnite|GetFortPlayerControllerfromActor (Variables|GAVars|Data|GetOwningPlayerPawn))))
  (Variables|Default|SetInventory%20Items _items))

(fn GetOwnedGameplayTags
  (params (self2 :type Gameplay%20Tag%20Asset%20Interface%20Interface))
  (return (GameplayTags|GetOwnedGameplayTags self2 :!class /Script/GameplayTags.GameplayTagAssetInterface)))
